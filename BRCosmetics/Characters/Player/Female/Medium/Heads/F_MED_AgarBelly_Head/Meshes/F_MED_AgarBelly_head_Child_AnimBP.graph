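# Child Anim Blueprint for the AgarBelly female head part, handling aim-time head shrinking, rotation correction and eyelash blinks.
# While the local player targets with a visible weapon, head and neck_02 scale down, apparently to clear the crosshair; certain weapon animation types are exempt.
# Both bones also receive negated, percentage-scaled deltas from ref pose as additive rotation.
# Blink values remapped from the eye material drive per-side eyelash sequence evaluators, layered per bone.
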AnimBlueprint /BRCosmetics/Characters/Player/Female/Medium/Heads/F_MED_AgarBelly_Head/Meshes/F_MED_AgarBelly_head_Child_AnimBP
target_skeleton=/Game/Characters/Player/Common/Fortnite_Base_Head/Fortnite_Base_Head_Export_Skeleton
graphs [
 AnimGraph {nodes [
   "Transform (Modify) Bone - Bone: neck_02"!(ComponentPose, bAlphaBoolEnabled:bool=True, Translation="0,0,0", Rotation="0,0,0", Scale="1,0.9,0.9") -> (Pose) {
    alphaInputType=Bool
    scaleMode=BMM_Replace
    scaleSpace=BCS_BoneSpace
    alphaBoolBlend {blendInTime=0.1, blendOption=Linear, blendOutTime=0.12}
    boneToModify {boneName=neck_02}
    scale {y=0.9, z=0.9}
   }
   n1 = "Transform (Modify) Bone - Bone: head"!(ComponentPose, bAlphaBoolEnabled:bool=false, Translation="0,0,0", Rotation="0,0,0", Scale="0.9,1,1") -> (Pose) {
    alphaInputType=Bool
    scaleMode=BMM_Replace
    scaleSpace=BCS_BoneSpace
    alphaBoolBlend {blendInTime=0.1, blendOption=Linear, blendOutTime=0.12}
    boneToModify {boneName=head}
    scale {x=0.9}
   }
   "Transform (Modify) Bone - Bone: neck_02"(ComponentPose=@n1, bAlphaBoolEnabled=True, Translation="0,0,0", Rotation="0,0,0", Scale="1,0.9,0.9") {
    alphaInputType=Bool
    scaleMode=BMM_Replace
    scaleSpace=BCS_BoneSpace
    alphaBoolBlend {blendInTime=0.1, blendOption=Linear, blendOutTime=0.12}
    boneToModify {boneName=neck_02}
    scale {y=0.9, z=0.9}
   }
   n2 = Get scaleDownAlpha() -> (scaleDownAlpha:bool)
   n3 = "Transform (Modify) Bone - Bone: head"(ComponentPose=Local To Component(LocalPose=Input Pose()), bAlphaBoolEnabled=@n2, Translation="0,0,-0.5", Rotation="0,0,0", Scale="0.9,1,0.9") {
    alphaInputType=Bool
    scaleMode=BMM_Replace
    scaleSpace=BCS_BoneSpace
    translationMode=BMM_Additive
    translationSpace=BCS_BoneSpace
    alphaBoolBlend {blendInTime=0.1, blendOption=Linear, blendOutTime=0.12}
    boneToModify {boneName=head}
    scale {x=0.9, z=0.9}
    translation {z=-0.5}
   }
   n4 = "Transform (Modify) Bone - Bone: neck_02"(ComponentPose=@n3, bAlphaBoolEnabled=@n2, Translation="0,0,0", Rotation="0,0,0", Scale="1,0.9,0.9") {
    alphaInputType=Bool
    scaleMode=BMM_Replace
    scaleSpace=BCS_BoneSpace
    translationMode=BMM_Additive
    translationSpace=BCS_BoneSpace
    alphaBoolBlend {blendInTime=0.1, blendOption=Linear, blendOutTime=0.12}
    boneToModify {boneName=neck_02}
    scale {y=0.9, z=0.9}
   }
   n5 = "Transform (Modify) Bone - Bone: head"!(ComponentPose=@n4, Alpha:float=1, Translation="0,0,0", Rotation=Get headRot() -> (headRot:Rotator), Scale="1,1,1") -> (Pose) {rotationMode=BMM_Additive, rotationSpace=BCS_BoneSpace, boneToModify {boneName=head}}
   n6 = Get neckRot() -> (neckRot:Rotator)
   n7 = "Sequence Evaluator (by time) 'Fortnite_Base_Head_AgarBelly_Skeleto'"!(ExplicitTime:float=Get R Blink Value() -> (R Blink Value:double)) -> (Pose:Pose Link Structure) {
    is=title AnimGraphNode_SequenceEvaluator
    sequence=/BRCosmetics/Characters/Player/Female/Medium/Heads/F_MED_AgarBelly_Head/Meshes/AgarBelly_EyelashPosing/Fortnite_Base_Head_AgarBelly_Skeleto
   }
   n8 = "Sequence Evaluator (by time) 'Fortnite_Base_Head_AgarBelly_Skeleto'"(ExplicitTime=Get L Blink Value() -> (L Blink Value:double)) {sequence=&Fortnite_Base_Head_AgarBelly_Skeleto}
   n9 = Layered blend per bone!(
    BasePose=Component To Local(ComponentPose="Transform (Modify) Bone - Bone: neck_02"!(ComponentPose=@n5, Alpha:float=1, Translation="0,0,0", Rotation=@n6, Scale="1,1,1") -> (Pose) {rotationMode=BMM_Additive, rotationSpace=BCS_BoneSpace, boneToModify {boneName=neck_02}})
    BlendPoses_0=@n8
    BlendPoses_1:Pose Link Structure=@n7
    BlendWeights_0
    BlendWeights_1:float=1
   ) -> (Pose) {
    blendPoses [{linkId=-1, sourceLinkId=-1}, {linkId=-1, sourceLinkId=-1}]
    blendWeights [1,1]
    layerSetup [
     {branchFilters [{boneName=def_eyelash_01_l}, {boneName=def_eyelash_02_l}, {boneName=def_eyelash_03_l}]}
     {branchFilters [{boneName=def_eyelash_01_r}, {boneName=def_eyelash_02_r}, {boneName=def_eyelash_03_r}]}
    ]
   }
   Output Pose(Result=@n9)
   "Sequence Evaluator (by time) 'Fortnite_Base_Head_AgarBelly_Skeleto'"(ExplicitTime=Get R Blink Lower Value() -> (R Blink Lower Value:double)) {sequence=&Fortnite_Base_Head_AgarBelly_Skeleto}
  ]}
 EventGraph {nodes [
   n1 = Cast To FortPlayerMannequin!(execute, Object=GetOwningActor()) -> (then, CastFailed, AsFort Player Mannequin:Fort Player Mannequin Object Reference) {is=title K2Node_DynamicCast}
   n2 = Get Dynamic Material Instance(execute=[@n1.then, @n4.then_0], SkeletalMeshComponent=GetPartSkeletalMeshComponent())
   > Set Eye Mat(Eye Mat:Material Instance Dynamic Object Reference=@n2) -> (Output_Get:Material Instance Dynamic Object Reference)
   Event BlueprintUpdateAnimation()
   > n3 = Sequence!(execute) -> (then_0, then_1, then_2:Exec, then_3:Exec)
   > n4 = Sequence!(execute) -> (then_0, then_1)
   n5 = Self-Reference()
   n6 = Get Eye Mat() -> (Eye Mat:Material Instance Dynamic Object Reference)
   n7 = Remap Material Blinks!(
    execute=@n4.then_1
    Character Part Anim Instance:Anim Instance Object Reference=@n5
    Dynamic Material Instance:Material Instance Dynamic Object Reference=@n6
    Squint Influence:double=0.35
    "Anim Length(Seconds)":double=3.3
   ) -> (then, L Blink Value:double, R Blink Value:double, L Blink Lower Value:double, R Blink Lower Value:double)
   > Set L Blink Value(L Blink Value:double=@n7) -> (Output_Get:double)
   > n8 = Remap Material Blinks(Character Part Anim Instance=@n5, Dynamic Material Instance=@n6, Squint Influence=1, "Anim Length(Seconds)"=3.3)
   > Set R Blink Value(R Blink Value:double=@n8.R Blink Value) -> (Output_Get:double)
   n9 = GetOwnerSkeletalMeshComponent()
   n10 = BreakRotator(InRot=BreakTransform(InTransform=GetDeltaTransformFromRefPose(self=@n9, BoneName=head, BaseName=neck_02)).Rotation)
   n11 = Get X_Percentage() -> (X_Percentage:double)
   n12 = Get Y_Percentage() -> (Y_Percentage:double)
   n13 = Get Z_Percentage() -> (Z_Percentage:double)
   Set headRot(
    execute=@n3.then_1
    headRot:Rotator=MakeRotator(
     Roll=float * float!(A=@n10, B=-1, C:double=@n11) -> (ReturnValue) {is=title K2Node_PromotableOperator}
     Pitch=float * float(A=@n10.Pitch, B=-1, C=@n12)
     Yaw=float * float(A=@n10.Yaw, B=-1, C=@n13)
    )
   ) -> (Output_Get:Rotator)
   n14 = BreakRotator(InRot=BreakTransform(InTransform=GetDeltaTransformFromRefPose(self=@n9, BoneName=neck_02, BaseName=neck_01)).Rotation)
   n15 = Get Z_Percentage()
   n16 = Get X_Percentage()
   n17 = Get Y_Percentage()
   > Set neckRot(neckRot:Rotator=MakeRotator(Roll=float * float(A=@n14, B=-1, C=@n16), Pitch=float * float(A=@n14.Pitch, B=-1, C=@n17), Yaw=float * float(A=@n14.Yaw, B=-1, C=@n15))) -> (Output_Get:Rotator)
   Set FortPawn(execute=@n3.then_2, FortPawn:Fort Pawn Object Reference=TryGetFortPawn()) -> (Output_Get:Fort Pawn Object Reference)
   n18 = Get FortPawn() -> (FortPawn:Fort Pawn Object Reference)
   > n19 = Is Valid(InputObject=@n18)
   Set weapon_visible(execute=@n19.Is Not Valid, weapon_visible:bool=false) -> (Output_Get:bool)
   n20 = Get CurrentWeapon!(self:Fort Pawn Object Reference=@n18) -> (CurrentWeapon:Fort Weapon Object Reference)
   n21 = GetCoreAnimation!(self:Fort Weapon Object Reference=@n20) -> (ReturnValue:EFortWeaponCoreAnimation Enum)
   n22 = Set notCoverCrosshair(
    execute=@n19.Is Valid
    notCoverCrosshair:bool=OR Boolean!(
     A="Equal (Enum)"(A:EFortWeaponCoreAnimation Enum=@n21, B:EFortWeaponCoreAnimation Enum=TrapPlacement)
     B="Equal (Enum)"(A=@n21, B=RocketLauncher)
     C:bool="Equal (Enum)"(A=@n21, B=GoingCommando)
    ) -> (ReturnValue)
   ) -> (Output_Get:bool)
   > n23 = Branch(Condition=@n22)
   > Set weapon_visible(weapon_visible=false)
   n24 = IsWeaponHolstered!(self:Fort Pawn Object Reference=@n18) -> (ReturnValue:bool)
   Set weapon_visible(execute=@n23.else, weapon_visible=AND Boolean(A=IsValid(Object=@n20), B=NOT Boolean(A=@n24)))
   n25 = Set isLocalPlayer(execute=@n3.then_3, isLocalPlayer:bool=IsLocalPawn!(Actor:Actor Object Reference=GetOwningActor()) -> (IsLocalPawn:bool)) -> (Output_Get:bool)
   > n26 = Branch(Condition=AND Boolean!(A=@n25, B=Get bIsTargeting(), C:bool=Get weapon_visible() -> (weapon_visible:bool)) -> (ReturnValue))
   > Set scaleDownAlpha(scaleDownAlpha:bool=true) -> (Output_Get:bool)
   Set scaleDownAlpha(execute=@n26.else, scaleDownAlpha=false)
   Event OnCharacterAndPickaxeCustomizationLoaded!() -> (
    OutputDelegate
    then
    CustomBaseMesh:Skeletal Mesh Component Object Reference
    CustomParts:Array of Custom Character Part Object References
    CustomMeshParts:Array of Skeletal Mesh Component Budgeted Object References
   ) {is=title K2Node_Event}
   > n27 = Get Dynamic Material Instance(SkeletalMeshComponent=GetPartSkeletalMeshComponent())
   > Set Eye Mat(Eye Mat=@n27)
   TryGetPawnOwner()
  ]}
]
variables [
 Eye Mat {type=MaterialInstanceDynamic}
 FortPawn {type=FortPawn, default=null}
 headRot {type=Rotator}
 isLocalPlayer {type=boolean}
 L Blink Value {type=number}
 neckRot {type=Rotator}
 notCoverCrosshair {type=boolean}
 R Blink Lower Value {type=number}
 R Blink Value {type=number}
 scaleDownAlpha {type=boolean}
 weapon_visible {type=boolean}
 X_Percentage {type=number}
 Y_Percentage {type=number}
 Z_Percentage {type=number}
]
